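# Sound cue for the ScooterRecon projectile's scanning orb opening, playing one of four random pulse waves in the weapons sound class.
# An IsFriendly branch selects the modulation: friendly plays at normal pitch with slight volume variation, while hostile or unset plays louder and higher-pitched.
# Both paths share the random wave node, with overridden HRTF attenuation, distance low-pass filtering, and local and global concurrency sets.
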SoundCue /ScooterGameplay/Sounds/ScooterRecon/ScooterRecon_Projectile_ScanningOrb_Open_Cue
root=Branch_0
nodes [
 {id=Branch_0, children [Modulator_0,Modulator_1,Modulator_1], properties {boolParameterName=IsFriendly}}
 {id=Modulator_0, children [Random_1], properties {pitchMax=1, pitchMin=1, volumeMax=1, volumeMin=0.8}}
 {id=Modulator_1, children [Random_1], properties {pitchMax=1.2, pitchMin=1.2, volumeMax=2, volumeMin=2}}
 {id=Random_1, children [WavePlayer_4,WavePlayer_5,WavePlayer_6,WavePlayer_7], properties {weights [1,1,1,1]}}
 {id=WavePlayer_4, properties {soundWaveAssetPtr=&/ScooterRecon_Pulse_C01}}
 {id=WavePlayer_5, properties {soundWaveAssetPtr=&/ScooterRecon_Pulse_C02}}
 {id=WavePlayer_6, properties {soundWaveAssetPtr=&/ScooterRecon_Pulse_C03}}
 {id=WavePlayer_7, properties {soundWaveAssetPtr=&/ScooterRecon_Pulse_C04}}
]
properties {
 bOverrideAttenuation=true
 maxDistance=3900
 soundClassObject=/Game/Sounds/FortSoundClassesAndModes/Fort_Weapons
 volumeMultiplier=0.15
 attenuationOverrides {
  bAttenuateWithLPF=true
  falloffDistance=3500
  lPFFrequencyAtMax=200
  lPFRadiusMax=3000
  lPFRadiusMin=800
  spatializationAlgorithm=SPATIALIZATION_HRTF
  stereoSpread=256
 }
 concurrencyOverrides {bLimitToOwner=true, maxCount=2}
 concurrencySet [
  &/ScooterRecon_Projectile_Local_Conc
  &/ScooterRecon_Projectile_Global_Conc
 ]
}
