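# Sound cue for the looping ground scanning sound of the ScooterRecon projectile.
# A looping wave player feeds a Modulator, then a SoundClass node that assigns it to Fort_Weapons.
# The cue plays quietly and slightly pitched up, with its own logarithmic attenuation adding distance low-pass filtering and stereo spread.
# A global ScooterRecon projectile concurrency set governs simultaneous instances.
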SoundCue /ScooterGameplay/Sounds/ScooterRecon/ScooterRecon_Projectile_Scanning_GroundLoop_Cue
root=SoundClass_3
nodes [
 {id=Modulator_3, children [WavePlayer_1]}
 {id=SoundClass_3, children [Modulator_3], properties {soundClassOverride=/Game/Sounds/FortSoundClassesAndModes/Fort_Weapons}}
 {
  id=WavePlayer_1
  properties {bLooping=true, soundWaveAssetPtr=&/ScooterRecon_Projectile_Scanning_GroundLoop_mix}
 }
]
properties {
 bOverrideAttenuation=true
 maxDistance=2755.5806
 pitchMultiplier=1.1
 priority=0.5
 volumeMultiplier=0.2
 attenuationOverrides {
  bAttenuateWithLPF=true
  distanceAlgorithm=Logarithmic
  falloffDistance=2355.5806
  lPFFrequencyAtMax=600
  lPFRadiusMax=3000
  lPFRadiusMin=800
  reverbWetLevelMax=0.7
  reverbWetLevelMin=0.2
  stereoSpread=1000
 }
 concurrencyOverrides {resolutionRule=PreventNew}
 concurrencySet [&/ScooterRecon_Projectile_Global_Conc]
}
